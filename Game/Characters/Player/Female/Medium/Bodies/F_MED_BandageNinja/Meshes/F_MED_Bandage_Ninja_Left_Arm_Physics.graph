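# Physics asset for the left-arm mesh of the medium female Bandage Ninja player body, rigging a dangling bandage chain.
# Capsule bodies cover torso, legs, clavicles and left arm, with no right-arm bodies.
# Four bandage bodies chain from a small master body on the left upper arm, with tight twist limits and mostly wide swing.
# A right master body hangs off the right clavicle without a chain.
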PhysicsAsset /Game/Characters/Player/Female/Medium/Bodies/F_MED_BandageNinja/Meshes/F_MED_Bandage_Ninja_Left_Arm_Physics
bodies {
 calf_l [{shapeType=Capsule, length=34.6826, radius=9.482, center=<-19.9486 -2.8284 0.1383>, rotation=<rot 0 -91.1589 -92.0231>}]
 calf_r [{shapeType=Capsule, length=34.6826, radius=9.482, center=<19.998 2.8274 0.0937>, rotation=<rot 0 88.8411 -87.0231>}]
 clavicle_l [{shapeType=Capsule, length=6.6964, radius=7.3853, center=<3.9138 -0.8097 -2.9272>, rotation=<rot -88.3257 102.4929 82.4543>}]
 clavicle_r [{shapeType=Capsule, length=6.6964, radius=7.3853, center=<-3.4752 0.5305 3.3858>, rotation=<rot -88.3282 -77.5052 82.4562>}]
 dyn_bandage_01_l [{shapeType=Capsule, length=6, radius=2, shapeName=dyn_bandage_01_l_capsule, center=<8 0 0>, rotation=<rot 90 0 0>}]
 dyn_bandage_02_l [{shapeType=Capsule, length=8, radius=3, shapeName=dyn_bandage_02_l_capsule, center=<6 0 0>, rotation=<rot 90 0 0>}]
 dyn_bandage_03_l [{shapeType=Capsule, length=2, radius=3, shapeName=dyn_bandage_03_l_capsule, center=<4 0 0>, rotation=<rot 90 0 0>}]
 dyn_bandage_04_l [{shapeType=Capsule, length=3, radius=3, shapeName=dyn_bandage_04_l_capsule, center=<4 0 0>, rotation=<rot 90 0 0>}]
 dyn_bandage_MASTER_l [{shapeType=Capsule, length=0.505, radius=0.505, rotation=<rot 0 -90 0>}]
 dyn_bandage_MASTER_r [{shapeType=Capsule, length=0.505, radius=0.505, rotation=<rot 0 -90 0>}]
 foot_l [{shapeType=Capsule, length=12.5982, radius=6.5127, center=<-2.9101 7.2748 -0.3501>, rotation=<rot 0 5.1558 -89.3972>}]
 foot_r [{shapeType=Capsule, length=12.6052, radius=6.5119, center=<2.9081 -7.2701 0.3337>, rotation=<rot 0 5.1452 -89.0835>}]
 hand_l [{shapeType=Capsule, length=6.1723, radius=6.0638, shapeName=hand_l_capsule, center=<7 0 0>, rotation=<rot 90 0 0>}]
 lowerarm_l [{shapeType=Capsule, length=18, radius=4.5, shapeName=lowerarm_l_capsule, center=<12 0 0>, rotation=<rot 90 0 0>}]
 pelvis [{shapeType=Capsule, length=5.809, radius=14.3783, center=<-4.524 -1.8634 -0.1967>, rotation=<rot -4.1601 -89.9995 -1.4037>}]
 spine_03 [{shapeType=Capsule, length=12.1237, radius=13.8935, center=<-4.5611 1.7894 -0.1524>, rotation=<rot 6.2753 -83.7246 89.9992>}]
 spine_05 [{shapeType=Capsule, length=17.2046, radius=12.0791, center=<-1.9814 1.858 0.7467>, rotation=<rot 0 -90 0>}]
 thigh_l [{shapeType=Capsule, length=30.6477, radius=10.4942, center=<-22.3366 -1.2788 -0.1036>, rotation=<rot 0 92.6888 -89.6929>}]
 thigh_r [{shapeType=Capsule, length=30.6477, radius=10.4942, center=<22.3407 1.2789 0.6548>, rotation=<rot 0 -87.3111 -89.6929>}]
 upperarm_l [{shapeType=Capsule, length=22.7328, radius=4.1826, center=<13.6692 0.2831 0.4097>, rotation=<rot 0.2555 91.1854 -87.1344>}]
}
constraints [
 {bone1Name=spine_03, bone2Name=pelvis, swing1LimitDegrees=45, swing2LimitDegrees=45, twistLimitDegrees=45}
 {bone1Name=spine_05, bone2Name=spine_03, swing1LimitDegrees=45, swing2LimitDegrees=45, twistLimitDegrees=45}
 {bone1Name=upperarm_l, bone2Name=clavicle_l, swing1LimitDegrees=45, swing2LimitDegrees=45, twistLimitDegrees=45}
 {bone1Name=thigh_l, bone2Name=pelvis, swing1LimitDegrees=45, swing2LimitDegrees=45, twistLimitDegrees=45}
 {bone1Name=calf_l, bone2Name=thigh_l, swing1LimitDegrees=45, swing2LimitDegrees=45, twistLimitDegrees=45}
 {bone1Name=thigh_r, bone2Name=pelvis, swing1LimitDegrees=45, swing2LimitDegrees=45, twistLimitDegrees=45}
 {bone1Name=calf_r, bone2Name=thigh_r, swing1LimitDegrees=45, swing2LimitDegrees=45, twistLimitDegrees=45}
 {
  bone1Name=clavicle_r
  bone2Name=spine_05
  swing1LimitDegrees=45
  swing1Motion=Free
  swing2LimitDegrees=45
  swing2Motion=Free
  twistLimitDegrees=45
  twistMotion=Free
 }
 {
  bone1Name=foot_l
  bone2Name=calf_l
  swing1LimitDegrees=45
  swing1Motion=Free
  swing2LimitDegrees=45
  swing2Motion=Free
  twistLimitDegrees=45
  twistMotion=Free
 }
 {
  bone1Name=foot_r
  bone2Name=calf_r
  swing1LimitDegrees=45
  swing1Motion=Free
  swing2LimitDegrees=45
  swing2Motion=Free
  twistLimitDegrees=45
  twistMotion=Free
 }
 {
  bone1Name=clavicle_l
  bone2Name=spine_05
  swing1LimitDegrees=45
  swing1Motion=Free
  swing2LimitDegrees=45
  swing2Motion=Free
  twistLimitDegrees=45
  twistMotion=Free
 }
 {
  bone1Name=dyn_bandage_MASTER_r
  bone2Name=clavicle_r
  swing1LimitDegrees=45
  swing1Motion=Free
  swing2LimitDegrees=45
  swing2Motion=Free
  twistLimitDegrees=45
  twistMotion=Free
 }
 {
  bone1Name=dyn_bandage_MASTER_l
  bone2Name=upperarm_l
  swing1LimitDegrees=45
  swing1Motion=Free
  swing2LimitDegrees=45
  swing2Motion=Free
  twistLimitDegrees=45
  twistMotion=Free
 }
 {bone1Name=dyn_bandage_01_l, bone2Name=dyn_bandage_MASTER_l, swing1LimitDegrees=45, swing2LimitDegrees=45, twistLimitDegrees=15}
 {bone1Name=dyn_bandage_02_l, bone2Name=dyn_bandage_01_l, swing1LimitDegrees=90, swing2LimitDegrees=90, twistLimitDegrees=5}
 {bone1Name=dyn_bandage_03_l, bone2Name=dyn_bandage_02_l, swing1LimitDegrees=90, swing2LimitDegrees=90, twistLimitDegrees=5}
 {bone1Name=dyn_bandage_04_l, bone2Name=dyn_bandage_03_l, swing1LimitDegrees=90, swing2LimitDegrees=90, twistLimitDegrees=5}
 {
  bone1Name=lowerarm_l
  bone2Name=upperarm_l
  swing1LimitDegrees=45
  swing1Motion=Free
  swing2LimitDegrees=45
  swing2Motion=Free
  twistLimitDegrees=45
  twistMotion=Free
 }
 {
  bone1Name=hand_l
  bone2Name=lowerarm_l
  swing1LimitDegrees=45
  swing1Motion=Free
  swing2LimitDegrees=45
  swing2Motion=Free
  twistLimitDegrees=45
  twistMotion=Free
 }
]
