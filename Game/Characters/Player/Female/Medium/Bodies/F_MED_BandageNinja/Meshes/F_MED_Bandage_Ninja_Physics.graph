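# Physics asset for the female medium Bandage Ninja player body, covering the skeleton with capsule bodies and adding jointed bandages on both arms.
# Each bandage is a four-segment chain hanging from a master body on the upper arm, with swing limits and tight twist limits between segments.
# Master-to-arm links and many body joints rotate freely, while hips, knees, shoulders and two spine joints use angular limits.
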PhysicsAsset /Game/Characters/Player/Female/Medium/Bodies/F_MED_BandageNinja/Meshes/F_MED_Bandage_Ninja_Physics
bodies {
 calf_l [{shapeType=Capsule, length=34.6826, radius=9.482, center=<-20 0 0>, rotation=<rot 90 0 0>}]
 calf_r [{shapeType=Capsule, length=34.6826, radius=9.482, center=<20 0 0>, rotation=<rot 90 0 0>}]
 clavicle_l [{shapeType=Capsule, length=6.6964, radius=7.3853, center=<3.9138 -0.8097 -2.9272>, rotation=<rot -88.3257 102.4929 82.4543>}]
 clavicle_r [{shapeType=Capsule, length=6.6964, radius=7.3853, center=<-3.4752 0.5305 3.3858>, rotation=<rot -88.3282 -77.5052 82.4562>}]
 dyn_bandage_01_l [{shapeType=Capsule, length=6.8785, radius=2.4086, shapeName=dyn_bandage_01_l_capsule, center=<7 -1 0>, rotation=<rot 90 0 0>}]
 dyn_bandage_01_r [{shapeType=Capsule, length=9.3781, radius=2.4801, shapeName=dyn_bandage_01_r_capsule, center=<9 -1 0>, rotation=<rot 90 0 0>}]
 dyn_bandage_02_l [{shapeType=Capsule, length=6.7028, radius=2.7476, shapeName=dyn_bandage_02_l_capsule, center=<6 0 0>, rotation=<rot 90 0 0>}]
 dyn_bandage_02_r [{shapeType=Capsule, length=8.8507, radius=2.782, shapeName=dyn_bandage_02_r_capsule, center=<8 0 0>, rotation=<rot 90 0 0>}]
 dyn_bandage_03_l [{shapeType=Capsule, length=3.068, radius=2.1731, shapeName=dyn_bandage_03_l_capsule, center=<4 0 0>, rotation=<rot 90 0 0>}]
 dyn_bandage_03_r [{shapeType=Capsule, length=5.7534, radius=1.961, shapeName=dyn_bandage_03_r_capsule, center=<6 0 0>, rotation=<rot 90 0 0>}]
 dyn_bandage_04_l [{shapeType=Capsule, length=4.1556, radius=1.361, shapeName=dyn_bandage_04_l_capsule, center=<4 0 0>, rotation=<rot 90 0 0>}]
 dyn_bandage_04_r [{shapeType=Capsule, length=8.3813, radius=1.1111, shapeName=dyn_bandage_04_r_capsule, center=<6 0 0>, rotation=<rot 90 0 0>}]
 dyn_bandage_MASTER_l [{shapeType=Capsule, length=0.505, radius=0.505, rotation=<rot 0 -90 0>}]
 dyn_bandage_MASTER_r [{shapeType=Capsule, length=0.505, radius=0.505, rotation=<rot 0 -90 0>}]
 foot_l [{shapeType=Capsule, length=12.5982, radius=6.5127, center=<-3 7 0>, rotation=<rot 0 0 90>}]
 foot_r [{shapeType=Capsule, length=12.6052, radius=6.5119, center=<3 -7 0>, rotation=<rot 0 0 90>}]
 hand_l [{shapeType=Capsule, length=6.1723, radius=6.0638, shapeName=hand_l_capsule, center=<7 0 0>, rotation=<rot 90 0 0>}]
 hand_r [{shapeType=Capsule, length=6.1723, radius=6.0638, shapeName=hand_r_capsule, center=<-6 0 0>, rotation=<rot 90 0 0>}]
 head [{shapeType=Capsule, length=0, radius=18, center=<9.0697 0.0746 0>, rotation=<rot 0 -90 0>}]
 lowerarm_l [{shapeType=Capsule, length=18, radius=4.5, shapeName=lowerarm_l_capsule, center=<12 0 0>, rotation=<rot 90 0 0>}]
 lowerarm_r [{shapeType=Capsule, length=18, radius=4.5, shapeName=lowerarm_r_capsule, center=<-12 0 0>, rotation=<rot 90 0 0>}]
 neck_02 [{shapeType=Capsule, length=0.505, radius=10.755, center=<-2.9035 0.6558 0>, rotation=<rot 0 -90 0>}]
 pelvis [{shapeType=Capsule, length=7.8686, radius=12.3188, center=<-4.524 -1.8634 -0.1967>, rotation=<rot -4.1601 -89.9995 -1.4037>}]
 spine_01 [{shapeType=Capsule, length=8.7738, radius=11.3174, shapeName=spine_01_capsule, center=<-2.891 0 0>}]
 spine_02 [{shapeType=Capsule, length=5.3128, radius=8.8172, shapeName=spine_02_capsule}]
 spine_03 [{shapeType=Capsule, length=7.5296, radius=9.8074, shapeName=spine_03_capsule}]
 spine_04 [{shapeType=Capsule, length=3.8841, radius=12.0095, shapeName=spine_04_capsule}]
 spine_05 [{shapeType=Capsule, length=4.7863, radius=12.0791, center=<-1.9814 1.858 0.7467>, rotation=<rot 0 -90 0>}]
 thigh_l [{shapeType=Capsule, length=32.0792, radius=10.0685, center=<-22 0 0>, rotation=<rot 85 180 180>}]
 thigh_r [{shapeType=Capsule, length=30.6477, radius=10.3185, center=<22 0 0>, rotation=<rot 85 180 180>}]
 upperarm_l [{shapeType=Capsule, length=22.7328, radius=4.1826, center=<13.6692 0.2831 0.4097>, rotation=<rot 0.2555 91.1854 -87.1344>}]
 upperarm_r [{shapeType=Capsule, length=20.1289, radius=4, center=<-12.334 0.1139 -0.0744>, rotation=<rot 0 -89.9999 -91.6879>}]
}
constraints [
 {bone1Name=spine_03, bone2Name=spine_02, swing1LimitDegrees=45, swing2LimitDegrees=45, twistLimitDegrees=45}
 {bone1Name=spine_05, bone2Name=spine_04, swing1LimitDegrees=45, swing2LimitDegrees=45, twistLimitDegrees=45}
 {bone1Name=upperarm_l, bone2Name=clavicle_l, swing1LimitDegrees=45, swing2LimitDegrees=45, twistLimitDegrees=45}
 {bone1Name=upperarm_r, bone2Name=clavicle_r, swing1LimitDegrees=45, swing2LimitDegrees=45, twistLimitDegrees=45}
 {bone1Name=thigh_l, bone2Name=pelvis, swing1LimitDegrees=45, swing2LimitDegrees=45, twistLimitDegrees=45}
 {bone1Name=calf_l, bone2Name=thigh_l, swing1LimitDegrees=45, swing2LimitDegrees=45, twistLimitDegrees=45}
 {bone1Name=thigh_r, bone2Name=pelvis, swing1LimitDegrees=45, swing2LimitDegrees=45, twistLimitDegrees=45}
 {bone1Name=calf_r, bone2Name=thigh_r, swing1LimitDegrees=45, swing2LimitDegrees=45, twistLimitDegrees=45}
 {
  bone1Name=clavicle_r
  bone2Name=spine_05
  swing1LimitDegrees=45
  swing1Motion=Free
  swing2LimitDegrees=45
  swing2Motion=Free
  twistLimitDegrees=45
  twistMotion=Free
 }
 {
  bone1Name=foot_l
  bone2Name=calf_l
  swing1LimitDegrees=45
  swing1Motion=Free
  swing2LimitDegrees=45
  swing2Motion=Free
  twistLimitDegrees=45
  twistMotion=Free
 }
 {
  bone1Name=foot_r
  bone2Name=calf_r
  swing1LimitDegrees=45
  swing1Motion=Free
  swing2LimitDegrees=45
  swing2Motion=Free
  twistLimitDegrees=45
  twistMotion=Free
 }
 {
  bone1Name=clavicle_l
  bone2Name=spine_05
  swing1LimitDegrees=45
  swing1Motion=Free
  swing2LimitDegrees=45
  swing2Motion=Free
  twistLimitDegrees=45
  twistMotion=Free
 }
 {
  bone1Name=neck_02
  bone2Name=spine_05
  swing1LimitDegrees=45
  swing1Motion=Free
  swing2LimitDegrees=45
  swing2Motion=Free
  twistLimitDegrees=45
  twistMotion=Free
 }
 {
  bone1Name=head
  bone2Name=neck_02
  swing1LimitDegrees=45
  swing1Motion=Free
  swing2LimitDegrees=45
  swing2Motion=Free
  twistLimitDegrees=45
  twistMotion=Free
 }
 {
  bone1Name=dyn_bandage_MASTER_r
  bone2Name=upperarm_r
  swing1LimitDegrees=45
  swing1Motion=Free
  swing2LimitDegrees=45
  swing2Motion=Free
  twistLimitDegrees=45
  twistMotion=Free
 }
 {
  bone1Name=dyn_bandage_MASTER_l
  bone2Name=upperarm_l
  swing1LimitDegrees=45
  swing1Motion=Free
  swing2LimitDegrees=45
  swing2Motion=Free
  twistLimitDegrees=45
  twistMotion=Free
 }
 {
  bone1Name=lowerarm_r
  bone2Name=upperarm_r
  swing1LimitDegrees=45
  swing1Motion=Free
  swing2LimitDegrees=45
  swing2Motion=Free
  twistLimitDegrees=45
  twistMotion=Free
 }
 {
  bone1Name=lowerarm_l
  bone2Name=upperarm_l
  swing1LimitDegrees=45
  swing1Motion=Free
  swing2LimitDegrees=45
  swing2Motion=Free
  twistLimitDegrees=45
  twistMotion=Free
 }
 {
  bone1Name=hand_l
  bone2Name=lowerarm_l
  swing1LimitDegrees=45
  swing1Motion=Free
  swing2LimitDegrees=45
  swing2Motion=Free
  twistLimitDegrees=45
  twistMotion=Free
 }
 {
  bone1Name=hand_r
  bone2Name=lowerarm_r
  swing1LimitDegrees=45
  swing1Motion=Free
  swing2LimitDegrees=45
  swing2Motion=Free
  twistLimitDegrees=45
  twistMotion=Free
 }
 {
  bone1Name=spine_04
  bone2Name=spine_03
  swing1LimitDegrees=45
  swing1Motion=Free
  swing2LimitDegrees=45
  swing2Motion=Free
  twistLimitDegrees=45
  twistMotion=Free
 }
 {
  bone1Name=spine_02
  bone2Name=spine_01
  swing1LimitDegrees=45
  swing1Motion=Free
  swing2LimitDegrees=45
  swing2Motion=Free
  twistLimitDegrees=45
  twistMotion=Free
 }
 {
  bone1Name=spine_01
  bone2Name=pelvis
  swing1LimitDegrees=45
  swing1Motion=Free
  swing2LimitDegrees=45
  swing2Motion=Free
  twistLimitDegrees=45
  twistMotion=Free
 }
 {bone1Name=dyn_bandage_01_l, bone2Name=dyn_bandage_MASTER_l, swing1LimitDegrees=75, swing2LimitDegrees=75, twistLimitDegrees=5}
 {bone1Name=dyn_bandage_02_l, bone2Name=dyn_bandage_01_l, swing1LimitDegrees=45, swing2LimitDegrees=45, twistLimitDegrees=5}
 {bone1Name=dyn_bandage_03_l, bone2Name=dyn_bandage_02_l, swing1LimitDegrees=45, swing2LimitDegrees=45, twistLimitDegrees=5}
 {bone1Name=dyn_bandage_04_l, bone2Name=dyn_bandage_03_l, swing1LimitDegrees=45, swing2LimitDegrees=45, twistLimitDegrees=5}
 {bone1Name=dyn_bandage_01_r, bone2Name=dyn_bandage_MASTER_r, swing1LimitDegrees=45, swing2LimitDegrees=45, twistLimitDegrees=5}
 {bone1Name=dyn_bandage_02_r, bone2Name=dyn_bandage_01_r, swing1LimitDegrees=45, swing2LimitDegrees=45, twistLimitDegrees=5}
 {bone1Name=dyn_bandage_03_r, bone2Name=dyn_bandage_02_r, swing1LimitDegrees=45, swing2LimitDegrees=45, twistLimitDegrees=5}
 {bone1Name=dyn_bandage_04_r, bone2Name=dyn_bandage_03_r, swing1LimitDegrees=45, swing2LimitDegrees=45, twistLimitDegrees=5}
]
