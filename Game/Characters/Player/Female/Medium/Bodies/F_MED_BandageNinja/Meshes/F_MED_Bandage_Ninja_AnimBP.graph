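# Post-process style AnimBP for the female Bandage Ninja body part, adding secondary motion to bandages and a hip pouch.
# It copies the parent mesh pose, aims bandage master bones at sockets, corrects the pouch by thigh crouch and caches the result.
# Bool blends pick per-arm rigid bodies when skydiving, otherwise cloth or full-body rigid body with gravity override, and bone offsets plus trails when rigid bodies are disabled.
# The computed skydiving force is not wired into the simulation.
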AnimBlueprint /Game/Characters/Player/Female/Medium/Bodies/F_MED_BandageNinja/Meshes/F_MED_Bandage_Ninja_AnimBP
target_skeleton=&/F_MED_Bandage_Ninja_Skeleton
graphs [
 AnimGraph {nodes [
   RigidBody() {
    bClampLinearTranslationLimitToRefPose=true
    overridePhysicsAsset=&/F_MED_Bandage_Ninja_Physics
    simulationSpace=BaseBoneSpace
    alphaBoolBlend {blendOption=Linear}
    baseBoneRef {boneName=root}
    componentAppliedLinearAccClamp=<2000 2000 2000>
    componentLinearAccScale=<1 1 1>
    componentLinearVelScale=<1 1 1>
   }
   n1 = RigidBody!(
    ComponentPose=Local To Component(LocalPose=Use cached pose 'copypose_cache'())
    Alpha
    OverrideWorldGravity:Vector=Get GravityOverride()
    ExternalForce
   ) -> (Pose) {
    bClampLinearTranslationLimitToRefPose=true
    bOverrideWorldGravity=true
    overridePhysicsAsset=&F_MED_Bandage_Ninja_Physics
    simulationSpace=BaseBoneSpace
    alphaBoolBlend {blendOption=Linear}
    baseBoneRef {boneName=root}
    componentAppliedLinearAccClamp=<3500 3500 4200>
    componentLinearAccScale=<0.5 0.5 0.5>
    componentLinearVelScale=<0.15 0.15 0.225>
    overrideWorldGravity {z=-980}
   }
   n2 = Blend Poses by bool(BlendPose_0=Use cached pose 'copypose_cache'(), BlendPose_1=Component To Local(ComponentPose=@n1), bActiveValue=Get bClothEnabled()) {blendType=Linear}
   n3 = RigidBody!(ComponentPose=Local To Component(LocalPose=Use cached pose 'copypose_cache'()), Alpha, ExternalForce) -> (Pose) {
    bClampLinearTranslationLimitToRefPose=true
    bOverrideWorldGravity=true
    overridePhysicsAsset=&/F_MED_Bandage_Ninja_Left_Arm_Physics
    simulationSpace=BaseBoneSpace
    alphaBoolBlend {blendOption=Linear}
    baseBoneRef {boneName=pelvis}
    componentAppliedLinearAccClamp=<3500 3500 4200>
    componentLinearAccScale=<0.5 0.5 0.5>
    componentLinearVelScale=<0.15 0.15 0.225>
    overrideWorldGravity {z=-980}
   }
   n4 = RigidBody(ComponentPose=@n3) {
    bClampLinearTranslationLimitToRefPose=true
    bOverrideWorldGravity=true
    overridePhysicsAsset=&/F_MED_Bandage_Ninja_Right_Arm_Physics
    simulationSpace=BaseBoneSpace
    alphaBoolBlend {blendOption=Linear}
    baseBoneRef {boneName=pelvis}
    componentAppliedLinearAccClamp=<3500 3500 4200>
    componentLinearAccScale=<0.5 0.5 0.5>
    componentLinearVelScale=<0.15 0.15 0.225>
    overrideWorldGravity {z=-980}
   }
   n5 = Component To Local(ComponentPose=@n4)
   n6 = Get Is_Sky_Diving_EG() -> (Is_Sky_Diving_EG:bool)
   n7 = "Transform (Modify) Bone - Bone: dyn_bandage_01_l"!(ComponentPose=Local To Component(LocalPose=Use cached pose 'copypose_cache'()), Alpha=1, Rotation="35,60,5") -> (Pose) {
    rotationMode=BMM_Additive
    rotationSpace=BCS_BoneSpace
    translationSpace=BCS_ParentBoneSpace
    alphaBoolBlend {blendInTime=0.7, blendOption=Linear, blendOutTime=0.5}
    alphaScaleBiasClamp {bInterpResult=true, interpSpeedDecreasing=5, interpSpeedIncreasing=7}
    boneToModify {boneName=dyn_bandage_01_l}
    rotation=<rot 35 60 5>
    translation {x=0.5}
   }
   n8 = "Transform (Modify) Bone - Bone: dyn_bandage_01_r"!(ComponentPose=@n7, Alpha=1, Rotation="-35,60,5") -> (Pose) {
    rotationMode=BMM_Additive
    rotationSpace=BCS_BoneSpace
    translationSpace=BCS_ParentBoneSpace
    alphaBoolBlend {blendInTime=0.7, blendOption=Linear, blendOutTime=0.5}
    alphaScaleBiasClamp {bInterpResult=true, interpSpeedDecreasing=5, interpSpeedIncreasing=7}
    boneToModify {boneName=dyn_bandage_01_r}
    rotation=<rot -35 60 5>
    translation {x=0.5}
   }
   n9 = "Transform (Modify) Bone - Bone: dyn_bandage_03_l"!(ComponentPose=@n8, Alpha=1, Rotation="0,20,0") -> (Pose) {
    rotationMode=BMM_Additive
    rotationSpace=BCS_BoneSpace
    translationSpace=BCS_ParentBoneSpace
    alphaBoolBlend {blendInTime=0.7, blendOption=Linear, blendOutTime=0.5}
    alphaScaleBiasClamp {bInterpResult=true, interpSpeedDecreasing=5, interpSpeedIncreasing=7}
    boneToModify {boneName=dyn_bandage_03_l}
    rotation {yaw=20}
    translation {x=0.5}
   }
   n10 = "Transform (Modify) Bone - Bone: dyn_bandage_03_r"!(ComponentPose=@n9, Alpha=1, Rotation="0,20,0") -> (Pose) {
    rotationMode=BMM_Additive
    rotationSpace=BCS_BoneSpace
    translationSpace=BCS_ParentBoneSpace
    alphaBoolBlend {blendInTime=0.7, blendOption=Linear, blendOutTime=0.5}
    alphaScaleBiasClamp {bInterpResult=true, interpSpeedDecreasing=5, interpSpeedIncreasing=7}
    boneToModify {boneName=dyn_bandage_03_r}
    rotation {yaw=20}
    translation {x=0.5}
   }
   n11 = Trail controller - Bone: dyn_bandage_04_l(ComponentPose=@n10, Alpha=1) {
    chainLength=3
    relaxationSpeedScale=2
    rotationLimits [{limitMax=<180 180 180>}, {limitMax=<180 180 180>}, {limitMax=<180 180 180>}]
    rotationOffsets [<0 0 0>, <0 0 0>, <0 0 0>]
    trailBone {boneName=dyn_bandage_04_l}
   }
   n12 = Trail controller - Bone: dyn_bandage_04_r(ComponentPose=@n11, Alpha=1) {
    chainLength=3
    relaxationSpeedScale=2
    rotationLimits [{limitMax=<180 180 180>}, {limitMax=<180 180 180>}, {limitMax=<180 180 180>}]
    rotationOffsets [<0 0 0>, <0 0 0>, <0 0 0>]
    trailBone {boneName=dyn_bandage_04_r}
   }
   Output Pose(Result=Blend Poses by bool(
     BlendPose_0=Blend Poses by bool(BlendPose_0=@n5, BlendPose_1=@n2, bActiveValue=@n6) {blendType=Linear}
     BlendPose_1=Component To Local(ComponentPose=@n12)
     bActiveValue=Get bRigidBodyEnabled()
    ) {blendType=Linear}) {layerGroup=None}
   n13 = Look At - Bone: dyn_bandage_MASTER_l(ComponentPose=Local To Component(LocalPose=Copy Pose From Mesh() {bUseAttachedParent=true}), Alpha=1) {
    alphaBoolBlend {blendOption=Linear}
    boneToModify {boneName=dyn_bandage_MASTER_l}
    lookAtLocation {x=0}
    lookAtTarget {bUseSocket=true, socketReference {socketName=aim_bandage_l}, boneReference=~}
    lookAt_Axis {axis=<1 0>}
    lookUp_Axis {axis=<0 1>}
   }
   n14 = Look At - Bone: dyn_bandage_MASTER_r(ComponentPose=@n13, Alpha=1) {
    alphaBoolBlend {blendOption=Linear}
    boneToModify {boneName=dyn_bandage_MASTER_r}
    lookAtLocation {x=0}
    lookAtTarget {bUseSocket=true, socketReference {socketName=aim_bandage_r}, boneReference=~}
    lookAt_Axis {axis=<1 0>}
    lookUp_Axis {axis=<0 1>}
   }
   n15 = "Transform (Modify) Bone - Bone: dyn_hip_pouch_l"!(ComponentPose=@n14, bAlphaBoolEnabled:bool=True, Translation="0.5,0,0", Rotation="12,0,0", Scale="1,1,1") -> (Pose) {
    alphaInputType=Bool
    rotationMode=BMM_Additive
    rotationSpace=BCS_ParentBoneSpace
    translationMode=BMM_Additive
    translationSpace=BCS_ParentBoneSpace
    alphaBoolBlend {blendOption=Linear}
    boneToModify {boneName=dyn_hip_pouch_l}
    rotation {pitch=12}
    translation {x=0.5}
   }
   n16 = "Transform (Modify) Bone - Bone: dyn_hip_pouch_l"!(
    ComponentPose=@n15
    Alpha:float=Get thigh_l_crouch() -> (thigh_l_crouch:double)
    Translation="-0.177,-4.996,0.000"
    Rotation="9.127,-0.169,25.312"
    Scale="1,1,1"
   ) -> (Pose) {
    rotationMode=BMM_Additive
    rotationSpace=BCS_ParentBoneSpace
    translationMode=BMM_Additive
    translationSpace=BCS_ParentBoneSpace
    alphaBoolBlend {blendOption=Linear}
    boneToModify {boneName=dyn_hip_pouch_l}
    rotation=<rot 9.127 -0.169 25.312>
    translation=<-0.177 -4.996>
   }
   Save cached pose 'copypose_cache'(Pose=Component To Local(ComponentPose=@n16)) {cacheName=copypose_cache}
   n17 = Spring controller - Bone: dyn_bandage_02_l!(ComponentPose:Component Space Pose Link Structure, Alpha:float=1) -> (Pose:Component Space Pose Link Structure) {
    is=title AnimGraphNode_SpringBone
    bLimitDisplacement=true
    bRotateY=true
    bRotateZ=true
    bTranslateX=false
    bTranslateZ=false
    maxDisplacement=2
    springStiffness=300
    alphaBoolBlend {blendOption=Linear}
    springBone {boneName=dyn_bandage_02_l}
   }
   n18 = Spring controller - Bone: dyn_bandage_02_l(ComponentPose=@n17, Alpha=1) {
    bLimitDisplacement=true
    bRotateY=true
    bRotateZ=true
    bTranslateX=false
    bTranslateZ=false
    maxDisplacement=2
    springStiffness=300
    alphaBoolBlend {blendOption=Linear}
    springBone {boneName=dyn_bandage_02_l}
   }
   n19 = Spring controller - Bone: dyn_bandage_02_r!(ComponentPose:Component Space Pose Link Structure=@n18, Alpha:float=1) -> (Pose:Component Space Pose Link Structure) {
    is=title AnimGraphNode_SpringBone
    bLimitDisplacement=true
    bRotateY=true
    bRotateZ=true
    bTranslateX=false
    bTranslateZ=false
    maxDisplacement=2
    springStiffness=300
    alphaBoolBlend {blendOption=Linear}
    springBone {boneName=dyn_bandage_02_r}
   }
   Spring controller - Bone: dyn_bandage_03_r!(ComponentPose:Component Space Pose Link Structure=@n19, Alpha:float=1) -> (Pose:Component Space Pose Link Structure) {
    is=title AnimGraphNode_SpringBone
    bLimitDisplacement=true
    bRotateY=true
    bRotateZ=true
    bTranslateX=false
    bTranslateZ=false
    maxDisplacement=2
    springStiffness=300
    alphaBoolBlend {blendOption=Linear}
    springBone {boneName=dyn_bandage_03_r}
   }
   Get skydiving vector force() -> (skydiving vector force:Vector)
  ]}
 EventGraph {nodes [
   Event BlueprintUpdateAnimation()
   > n1 = Sequence!(execute) -> (then_0, then_1, then_2:Exec, then_3:Exec, then_4:Exec)
   n2 = Branch(execute=@n1.then_2, Condition=Get bIsAccelerating2D())
   > Set Is_Accelerating_EG(Is_Accelerating_EG:bool=true) -> (Output_Get:bool)
   Set Is_Accelerating_EG(execute=@n2.else, Is_Accelerating_EG=false)
   n3 = Branch(execute=@n1.then_3, Condition=Get bIsSkyDiving())
   > Set Is_Sky_Diving_EG(Is_Sky_Diving_EG:bool=true) -> (Output_Get:bool)
   Set Is_Sky_Diving_EG(execute=@n3.else, Is_Sky_Diving_EG=false)
   n4 = Gravity Override State Function DEPRECATED(
    execute=@n1.then_4
    CharacterPartAnimInstance=Self-Reference()
    GameTimeInSeconds=GetGameTimeInSeconds()
    GravityOverrideParams=Get GravityOverrideParams() -> (GravityOverrideParams:Gravity Override Params Struct Structure)
   )
   > Set GravityOverride(GravityOverride=@n4)
   n5 = BreakTransform(InTransform=GetDeltaTransformFromRefPose(self=GetPartSkeletalMeshComponent(), BoneName=thigh_l, BaseName=None))
   Set thigh_l_crouch(execute=@n1.then_1, thigh_l_crouch:double=MapRangeClamped(Value=BreakRotator(InRot=@n5.Rotation).Yaw, InRangeA=-20, InRangeB=-100)) -> (Output_Get:double)
   n6 = Get bIsPlayingMeleeAnim()
   n7 = Get bIsMoving2D()
   > Set Melee_swing(Melee_swing:double=SelectFloat(A=1, B=0, bPickA=AND Boolean(A=@n6, B=@n7))) -> (Output_Get:double)
   Branch(execute=@n1.then_0, Condition=Get bIsSkyDiving())
   > n8 = Branch(Condition=Get bIsSkydiveFloating())
   n9 = RandomFloatInRangeFromStream!(Stream:"Random Stream Structure (by ref)"=Get random_stream() -> (random_stream:Random Stream Structure), Min:float=-20000, Max:float=20000) -> (ReturnValue:float)
   n10 = MakeVector(X=@n9, Y=@n9)
   > Set skydiving vector force(skydiving vector force:Vector=@n10) -> (Output_Get:Vector)
   n11 = Branch(execute=@n8.else, Condition=Get bIsSkydiveDiveMode())
   n12 = Get Direction Between Sockets!(
    Component:Skeletal Mesh Component Object Reference=GetOwnerSkeletalMeshComponent()
    SocketOrBoneNameFrom:Name=head
    SocketOrBoneNameTo:Name=pelvis
   ) -> (ReturnValue:Vector) {is=title K2Node_CallFunction}
   > Set skydiving vector force(skydiving vector force=vector * vector(A:Vector=vector * vector(A=@n10, B=@n12), B:Vector="0.2,0.2,0.2") -> (ReturnValue:Vector) {is=title K2Node_CommutativeAssociativeBinaryOperator})
   Set skydiving vector force(execute=@n11.else, skydiving vector force=@n10)
   Event BlueprintInitializeAnimation()
   > Set random_stream(random_stream:Random Stream Structure=MakeRandomStream!(InitialSeed:int=0) -> (ReturnValue:Random Stream Structure)) -> (Output_Get:Random Stream Structure)
   > n13 = Cast To ClothingSimulationInteractorNv!(
    execute
    Object=GetClothingSimulationInteractor!(self:Skeletal Mesh Component Object Reference=GetPartSkeletalMeshComponent(), ClothingAssetName:Name=None) -> (ReturnValue:Clothing Simulation Interactor Object Reference)
   ) -> (then, CastFailed, AsClothing Simulation Interactor Nv:Clothing Simulation Interactor Nv Object Reference) {is=title K2Node_DynamicCast}
   > Set clothing_sim_interactor(clothing_sim_interactor:Clothing Simulation Interactor Nv Object Reference=@n13) -> (Output_Get:Clothing Simulation Interactor Nv Object Reference)
   n14 = Reroute Node()
   n15 = BreakTransform(InTransform=GetDeltaTransformFromRefPose(self=GetPartSkeletalMeshComponent(), BoneName=clavicle_r, BaseName=None))
   Set collar_r_alpha(
    execute=[@n14.OutputPin, @n17.then]
    collar_r_alpha:double=MapRangeClamped(Value=BreakRotator(InRot=@n15.Rotation).Pitch, InRangeA=15, InRangeB=18)
   ) -> (Output_Get:double)
   Reroute Node()
   n16 = BreakTransform(InTransform=GetDeltaTransformFromRefPose(self=GetPartSkeletalMeshComponent(), BoneName=clavicle_l, BaseName=None))
   > n17 = Set collar_l_alpha(collar_l_alpha:double=MapRangeClamped(Value=BreakRotator(InRot=@n16.Rotation).Pitch, InRangeA=5, InRangeB=40)) -> (Output_Get:double)
  ]}
]
variables [
 clothing_sim_interactor {type=ClothingSimulationInteractorNv, default=null}
 collar_l_alpha {type=number}
 collar_r_alpha {type=number}
 GravityOverride {type=Vector}
 GravityOverrideParams {type=GravityOverrideParamsStruct}
 Is_Accelerating_EG {type=boolean}
 Is_Sky_Diving_EG {type=boolean}
 Melee_swing {type=number}
 random_stream {type=RandomStream, default {initialSeed=0, seed=0}}
 skydiving vector force {type=Vector}
 thigh_l_crouch {type=number}
]
